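# Material function that outputs world position offset and normals for a Landmass terrain mesh from the packed RG8 HeightRT texture.
# WPO unpacks height at landscape UVs from Landmass_TerrainMesh_Coords and scales it along Z using MPC_Landscape.
# A custom HLSL node builds normals from neighboring texel height differences at four texels and bilinearly blends them.
# An Emissive output renders red where height data exceeds the roughly ±32k range.
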MaterialFunction /Landmass/Landscape/BlueprintBrushes/MF/Landmass_TerrainMesh_WPO_Normals
graph {nodes [
  MaterialExpressionTextureObjectParameter_1 = TextureObjectParameter() {parameterName=HeightRT, texture=&Black}
  MaterialExpressionReroute_1 = Reroute(None=AppendVector(
    A=CollectionParameter() {collection=/Landmass/Landscape/BlueprintBrushes/MPC/MPC_Landscape, parameterName=RTResX}
    B=CollectionParameter() {collection=&MPC_Landscape, parameterName=RTResY}
   ))
  MaterialExpressionMaterialFunctionCall_0 = MaterialFunctionCall() {materialFunction=&/Landmass_TerrainMesh_Coords}
  MaterialExpressionReroute_0 = Reroute(None=ComponentMask(None=CollectionParameter() {collection=&MPC_Landscape, parameterName=LandscapeScale}) {r=true})
  MaterialExpressionCollectionParameter_4 = CollectionParameter() {collection=&MPC_Landscape, parameterName=LandscapeZScale}
  n1 = Custom!(
   HeightTex=@MaterialExpressionTextureObjectParameter_1
   TextureSize=@MaterialExpressionReroute_1
   UV=@MaterialExpressionMaterialFunctionCall_0.Landscape UVs
   GridSize=@MaterialExpressionReroute_0
   HeightScale=@MaterialExpressionCollectionParameter_4
  ) {
   code="//HeightTex: input Texture Object containing heights.R and water.B\r\n//TextureSize: input for size\r\n//GridSize: input float grid world size\r\n//HeightScale: input for height range scale\r\n\r\n\r\nstruct NormalSampleHelper\r\n{\r\n\r\n\tfloat2 Offsets[4];\r\n\tfloat2 TexelSize;\r\n\tfloat HeightScaleValue;\r\n\tfloat GridSizeValue;\r\n\r\n\tfloat UnpackRG8(float2 inSourceSample)\r\n\t{\r\n\t\treturn clamp(float(((int)round(inSourceSample.r * 255.0) << 8) | (int)round(inSourceSample.g * 255.0)), 0.0, 65536.0);\r\n\t}\r\n\t\r\n\tfloat3 SampleNormal(float2 inUV, Texture2D inTexture, SamplerState inTextureSampler)\r\n\t{\r\n\t\tfloat2 NearestUV = (floor(inUV / TexelSize)*TexelSize) + (TexelSize * 0.5);\r\n\r\n\t\tfloat2 PosYSample = UnpackRG8(inTexture.SampleLevel(inTextureSampler, NearestUV + Offsets[0] * TexelSize, 0).rg);\r\n\t\tfloat2 NegYSample = UnpackRG8(inTexture.SampleLevel(inTextureSampler, NearestUV + Offsets[1] * TexelSize, 0).rg);\r\n\t\tfloat2 PosXSample = UnpackRG8(inTexture.SampleLevel(inTextureSampler, NearestUV + Offsets[2] * TexelSize, 0).rg);\r\n\t\tfloat2 NegXSample = UnpackRG8(inTexture.SampleLevel(inTextureSampler, NearestUV + Offsets[3] * TexelSize, 0).rg);\r\n\t\t\r\n\t\tfloat3 Tangent =  float3( 2*GridSizeValue, 0, ((PosXSample.x) - (NegXSample.x))*HeightScaleValue);\r\n\t\tfloat3 Binormal = float3( 0, 2*GridSizeValue, ((PosYSample.x) - (NegYSample.x))*HeightScaleValue);\r\n\t\tTangent = normalize(Tangent);\r\n\t\tBinormal = normalize(Binormal);\r\n\t\t\r\n\t\treturn normalize(cross(Tangent, Binormal));\r\n\t}\r\n\t\r\n\tfloat3 SampleNormal_BilinearFilter(float2 inUV, Texture2D inTexture, SamplerState inTextureSampler)\r\n\t{\r\n\t\tinUV -= (TexelSize * 0.5);\r\n\t\r\n\t\tfloat2 NearestUV = (floor(inUV / TexelSize)*TexelSize) + (TexelSize * 0.5);\r\n\r\n\t\tfloat3 Sample_x0y0 = SampleNormal(NearestUV + float2(0,0) * TexelSize, inTexture, inTextureSampler);\r\n\t\tfloat3 Sample_x1y0 = SampleNormal(NearestUV + float2(1,0) * TexelSize, inTexture, inTextureSampler);\r\n\t\tfloat3 Sample_x0y1 = SampleNormal(NearestUV + float2(0,1) * TexelSize, inTexture, inTextureSampler);\r\n\t\tfloat3 Sample_x1y1 = SampleNormal(NearestUV + float2(1,1) * TexelSize, inTexture, inTextureSampler);\r\n\t\t\r\n\t\tfloat2 BlendPhase = frac( inUV / TexelSize );\r\n\t\t\r\n\t\tfloat3 Blend_X0 = lerp(Sample_x0y0, Sample_x1y0, BlendPhase.x);\r\n\t\tfloat3 Blend_X1 = lerp(Sample_x0y1, Sample_x1y1, BlendPhase.x);\r\n\t\tfloat3 OutNormal = lerp(Blend_X0, Blend_X1, BlendPhase.y);\r\n\t\t\r\n\t\treturn normalize(OutNormal);\r\n\t}\r\n\t\r\n};\r\n\r\nNormalSampleHelper Helper;\r\nHelper.Offsets[0] = float2(0,1);\r\nHelper.Offsets[1] = float2(0,-1); \r\nHelper.Offsets[2] = float2(1,0); \r\nHelper.Offsets[3] = float2(-1,0);\r\nHelper.TexelSize = 1 / TextureSize;\r\nHelper.HeightScaleValue = HeightScale;\r\nHelper.GridSizeValue = GridSize;\r\n\r\n\r\nreturn Helper.SampleNormal_BilinearFilter(UV, HeightTex, HeightTexSampler);\r\n//return Helper.SampleNormal(UV, HeightTex, HeightTexSampler);\r\n"
   description=Compute Normal
   inputs [
    {
     inputName=HeightTex
     input {expression=&self:MaterialExpressionTextureObjectParameter_1}
    }
    {
     inputName=TextureSize
     input {expression=&self:MaterialExpressionReroute_1}
    }
    {
     inputName=UV
     input {expression=&self:MaterialExpressionMaterialFunctionCall_0}
    }
    {inputName=GridSize, input {expression=&self:MaterialExpressionReroute_0}}
    {
     inputName=HeightScale
     input {expression=&self:MaterialExpressionCollectionParameter_4}
    }
   ]
  }
  FunctionOutput(None=NamedRerouteDeclaration(None=@n1) {name=Terrain Normal, nodeColor=<rgba 0.6875 1 0 1>}) {outputName=Terrain Normal, sortPriority=1}
  n2 = MaterialFunctionCall() {materialFunction=&Landmass_TerrainMesh_Coords}
  n3 = Multiply(
   A=MaterialFunctionCall!(RG=TextureSampleParameter2D(UVs=@n2.Landscape UVs) {parameterName=HeightRT, samplerSource=SSM_Wrap_WorldGroupSettings, texture=&Black}.RGB) {materialFunction=&/UnpackRG8Height}.Result
   B=CollectionParameter() {collection=&MPC_Landscape, parameterName=LandscapeZScale}
  )
  FunctionOutput(None=Multiply(A=NamedRerouteDeclaration(None=@n3) {name=Landscape Z, nodeColor=<rgba 0 1 0.0156 1>}, B=Constant3Vector() {constant=<rgba 0 0 1 0>})) {outputName=WPO Terrain, sortPriority=2}
  n4 = Divide(
   A=NamedRerouteUsage()
   B=CollectionParameter() {collection=&MPC_Landscape, parameterName=LandscapeZScale}
  )
  n5 = Subtract(A=@n4) {constB=32767}
  n6 = Subtract(A=Multiply(A=@n4) {constB=-1}) {constB=32767}
  n7 = Max(A=Saturate(None=Ceil(None=@n5)), B=Saturate(None=Ceil(None=@n6)))
  n8 = Constant3Vector() {constant=<rgba 1 0 0 0>}
  FunctionOutput(None=NamedRerouteDeclaration(None=Multiply(A=@n7, B=@n8)) {name=Clamping Z Range Error, nodeColor=<rgba 1 0 0.0398 1>}) {description=Renders a red color to show when the height data exceeds the max range of -32k to +32k., outputName=Emissive}
  MaterialFunctionCall() {materialFunction=/Landmass/Landscape/ErosionBrush/Materials/Functions/ErosionBrush_ParticleCoords}
 ]}
