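# Landmass preview material for a landscape terrain mesh, flagged for Nanite and instanced static meshes, with world-space normals.
# Landmass_TerrainMesh functions supply landscape UVs and a terrain World Position Offset term.
# Grass and Sand weightmap textures blend tint colors over a base value, with further weightmap color layers.
# A Debug Quadtree composite scales UVs by MPC_Landscape resolution, drawing numbered checker cells at two levels, blendable via Debug Checker Pattern.
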Material /Landmass/Landscape/BlueprintBrushes/Materials/Preview/Landmass_LandscapePreview_Mesh
graph {
 properties {
  bTangentSpaceNormal=false
  bUsedWithInstancedStaticMeshes=true
  bUsedWithNanite=true
  pixelDepthOffsetMode=PDOM_Legacy
  usedShadingModels=Default Lit
 }
 nodes [
  Composite() {subgraphName=Debug Quadtree}
  n1 = MaterialFunctionCall() {materialFunction=/Landmass/Landscape/BlueprintBrushes/MF/Landmass_TerrainMesh_Coords}
  n2 = LinearInterpolate(
   A=LinearInterpolate(
    A=ScalarParameter() {defaultValue=0.185, parameterName=Base}
    B=VectorParameter() {parameterName=Weightmap Color 1, defaultValue=<rgba 0.0604 0.1146 0.0246 1>}.RGB
    Alpha=TextureSampleParameter2D(UVs=@n1.Landscape UVs) {parameterName=Grass, texture=&Black}.R
   )
   B=VectorParameter() {parameterName=Weightmap Color 2, defaultValue=<rgba 0.9688 0.4848 0.2069 1>}.RGB
   Alpha=TextureSampleParameter2D(UVs=MaterialFunctionCall() {materialFunction=&Landmass_TerrainMesh_Coords}.Landscape UVs) {parameterName=Sand, texture=&Black}.R
  )
  n3 = MaterialFunctionCall() {materialFunction=&Landmass_TerrainMesh_Coords}
  n4 = CollectionParameter() {collection=/Landmass/Landscape/BlueprintBrushes/MPC/MPC_Landscape, parameterName=RTResX}
  n5 = CollectionParameter() {collection=&MPC_Landscape, parameterName=RTResY}
  n6 = Subtract(A=@n3.Landscape UVs, B=Divide(B=AppendVector(A=@n4, B=@n5)) {constA=0.5})
  n7 = Divide(
   A=AppendVector(
    A=CollectionParameter() {collection=&MPC_Landscape, parameterName=LSQuadsX}
    B=CollectionParameter() {collection=&MPC_Landscape, parameterName=LSQuadsY}
   )
   B=AppendVector(
    A=CollectionParameter() {collection=&MPC_Landscape, parameterName=RTResX}
    B=CollectionParameter() {collection=&MPC_Landscape, parameterName=RTResY}
   )
  )
  n8 = Power(Base=Constant() {r=2}, Exp=ScalarParameter() {defaultValue=2, parameterName=Quadtree Level B})
  n9 = Multiply(A=NamedRerouteDeclaration(None=Divide(A=@n6, B=@n7)) {name=Scaled UVs, nodeColor=<rgba 1 0.0469 0 1>}, B=@n8)
  n10 = Add(B=MaterialFunctionCall!(2D Index=Floor(None=@n9), Columns X=@n8) {materialFunction=/Engine/Functions/Engine_MaterialFunctions01/Coordinates/2Dto1DIndex}.1D Index) {constA=5}
  n11 = ScalarParameter() {defaultValue=2, parameterName=D}
  n12 = Frac(None=Add(A=@n9, B=Constant2Vector() {r=-0.2}))
  n13 = MaterialFunctionCall!(UVs=NamedRerouteUsage(), Tiling=Power(Base=Constant() {r=2}, Exp=ScalarParameter() {defaultValue=2, parameterName=Quadtree Level B})) {materialFunction=/Engine/ArtTools/RenderToTexture/MaterialFunctions/CheckerPattern}
  n14 = Constant3Vector() {constant=<rgba 0 0.5 0 0>}
  n15 = Power(Base=Constant() {r=2}, Exp=ScalarParameter() {defaultValue=3, parameterName=Quadtree Level})
  n16 = Multiply(A=NamedRerouteUsage(), B=@n15)
  n17 = Add(B=MaterialFunctionCall(2D Index=Floor(None=@n16), Columns X=@n15) {materialFunction=&2Dto1DIndex}.1D Index) {constA=21}
  n18 = ScalarParameter() {defaultValue=2, parameterName=D}
  n19 = Frac(None=Add(A=@n16, B=Constant2Vector() {r=-0.2}))
  n20 = MaterialFunctionCall(UVs=NamedRerouteUsage(), Tiling=Power(Base=Constant() {r=2}, Exp=ScalarParameter() {defaultValue=3, parameterName=Quadtree Level})) {materialFunction=&CheckerPattern}
  n21 = Constant3Vector() {constant=<rgba 0.5 0 0 0>}
  n22 = Add(
   A=Add(
    A=MaterialFunctionCall!(Number=@n10, MaximumNumberOfDigits=@n11, DebugTextLocation RG_UpperRight BA_LowerLeft, UVs=@n12) {materialFunction=/Engine/Functions/Engine_MaterialFunctions01/Debug/DebugIntValues}.Result
    B=Multiply(A=@n13.Result, B=@n14)
   )
   B=Add(
    A=MaterialFunctionCall(Number=@n17, MaximumNumberOfDigits=@n18, UVs=@n19) {materialFunction=&DebugIntValues}.Result
    B=Multiply(A=@n20.Result, B=@n21)
   )
  )
  MaterialExpressionReroute_0 = Reroute(None=@n22)
  LinearInterpolate(A=@n2, B=@MaterialExpressionReroute_0.Output, Alpha=ScalarParameter() {parameterName=Debug Checker Pattern})
  LinearInterpolate(
   A=LinearInterpolate(A=@n2, B=VectorParameter() {parameterName=Weightmap Color 3, defaultValue=<rgba 0.8 0.9 0.9 1>}.RGB, Alpha=NamedRerouteUsage())
   B=VectorParameter() {parameterName=Weightmap Color 4, defaultValue=<rgba 0.1 0.1 0.1 1>}.RGB
   Alpha=NamedRerouteUsage()
  )
  MaterialFunctionCall() {materialFunction=&Landmass_TerrainMesh_Coords}
  Multiply(
   A=MaterialFunctionCall() {materialFunction=/Landmass/Landscape/BlueprintBrushes/MF/Landmass_TerrainMesh_WPO_Normals}.WPO Terrain
   B=ScalarParameter() {defaultValue=1, parameterName=fasdfafsdasf}
  )
  Multiply(A=@n3.Landscape UVs)
  n23 = TextureSampleParameter2D(UVs=MaterialFunctionCall() {materialFunction=&Landmass_TerrainMesh_Coords}.Landscape UVs) {parameterName=Sand, texture=&Black}
  NamedRerouteDeclaration(None=Saturate(None=@n23.R)) {name=Weightmap 1, nodeColor=<rgba 1 0 0.8203 1>}
  NamedRerouteDeclaration(None=Saturate(None=@n23.G)) {name=Weightmap 2, nodeColor=<rgba 0.2891 0 1 1>}
  NamedRerouteDeclaration(None=Saturate(None=@n23.B)) {name=Weightmap 3, nodeColor=<rgba 0.9453 0 1 1>}
  NamedRerouteDeclaration(None=Saturate(None=OneMinus(None=@n23.A))) {name=Weightmap 4, nodeColor=<rgba 0 1 0.9766 1>}
  NamedRerouteUsage()
  NamedRerouteUsage()
  PinBase()
  PinBase!(Output=@n22) {reroutePins [Output {expression=&self:MaterialExpressionReroute_0}]}
  ScalarParameter()
  TextureSampleParameter2D(UVs=MaterialFunctionCall() {materialFunction=&Landmass_TerrainMesh_Coords}.Landscape UVs) {parameterName=HeightRT, samplerSource=SSM_Wrap_WorldGroupSettings, texture=&Black}
  VectorParameter() {parameterName=asdfasdf, defaultValue=<rgba 0 0.5 0 0>}
 ]
}
